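# Behavior tree, named for BT injection, that makes AI bots use the Rift Gauntlets' primary fire.
# A single sequence rolls a random use count between blackboard min and max, loops the use-item task on the injected item that many times, then waits with random deviation.
# A focus evaluator service on the root selector handles target focusing throughout.
# Use count, input hold duration and wait timing are bound to blackboard keys.
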BehaviorTree /RiftGauntletsGameplay/BTInjection/BT_UseItem_RiftGauntlets_Primary
blackboard=&/BB_UseItem_RiftGauntlets
root {
 name=Selector
 class=BTComposite_Selector
 children [UsePrimary {
   class=BTComposite_Sequence
   children [
    Determine Use Count {
     class=FortBTTask_RandomInt
     children []
     decorators []
     properties {
      bIgnoreRestartSelf=false
      nodeName=Determine Use Count
      blackboardKey {bNoneIsAllowedValue=false, selectedKeyName=Item_UseCount}
      maximumValue {defaultValue=0, key=Item_MaxUseCount}
      minimumValue {defaultValue=0, key=Item_MinUseCount}
     }
     services []
    }
    Primary Fire {
     class=FortAthenaBTTask_UseItem
     decorators [Loop {class=BTDecorator_Loop, properties {bInfiniteLoop=false, numLoops {key=Item_UseCount}}}]
     properties {
      actionObjectKeyName=AIEvaluator_Injection_Item
      executionStatusKeyName=None
      nodeName=Primary Fire
      finishAfterItemUse {defaultValue=true}
      inputHoldDuration {key=Item_InputHoldDuration}
      maximumWaitTimeBetweenUses {defaultValue=0}
      resetActionObjectKey {defaultValue=false}
     }
    }
    Wait {class=BTTask_Wait, properties {randomDeviation {key=Item_WaitTimeAfterUseDeviation}, waitTime {defaultValue=5, key=Item_WaitTimeAfterUse}}}
   ]
   decorators []
   properties {
    nodeName=UsePrimary
    children [
     {childTask=&self:FortBTTask_RandomInt_1}
     {
      childTask=&self:FortAthenaBTTask_UseItem_0
      decorators [&self:BTDecorator_Loop_1]
     }
     {childTask=&self:BTTask_Wait_4}
    ]
   }
  }]
 decorators []
 properties {
  children [{childComposite=&self:BTComposite_Sequence_9, childTask=null, decorators []}]
  services [&self:FortAthenaBTService_AIEvaluator_55]
 }
 services [Focus on target {
   class=FortAthenaBTService_AIEvaluator
   properties {interval=0.5, nodeName=Focus on target, aIEvaluatorClass {defaultValue=&FortAthenaAIBotEvaluator_HandleFocusing}}
  }]
}
