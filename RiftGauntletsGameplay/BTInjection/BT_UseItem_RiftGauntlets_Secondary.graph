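# Injected AI behavior tree that makes bots use the Rift Gauntlets' secondary fire.
# A root selector runs a focusing evaluator service to keep the bot on its target.
# Its single sequence fires the injected item in alternate mode, with hold duration read from the blackboard.
# It then waits for a blackboard-driven time with random deviation before finishing.
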BehaviorTree /RiftGauntletsGameplay/BTInjection/BT_UseItem_RiftGauntlets_Secondary
blackboard=&/BB_UseItem_RiftGauntlets
root {
 name=Selector
 class=BTComposite_Selector
 children [Use Secondary {
   class=BTComposite_Sequence
   children [
    Secondary Fire {
     class=FortAthenaBTTask_UseItem
     decorators []
     properties {
      actionObjectKeyName=AIEvaluator_Injection_Item
      executionStatusKeyName=None
      nodeName=Secondary Fire
      finishAfterItemUse {defaultValue=true}
      inputHoldDuration {key=Item_InputHoldDuration}
      maximumWaitTimeBetweenUses {defaultValue=0}
      resetActionObjectKey {defaultValue=false}
      useAlternateMode {defaultValue=true}
     }
    }
    Wait {class=BTTask_Wait, properties {randomDeviation {key=Item_WaitTimeAfterUseDeviation}, waitTime {defaultValue=5, key=Item_WaitTimeAfterUse}}}
   ]
   decorators []
   properties {
    nodeName=Use Secondary
    children [
     {childTask=&self:FortAthenaBTTask_UseItem_1}
     {childTask=&self:BTTask_Wait_3}
    ]
   }
  }]
 decorators []
 properties {
  children [{childComposite=&self:BTComposite_Sequence_7, childTask=null, decorators []}]
  services [&self:FortAthenaBTService_AIEvaluator_55]
 }
 services [Focus on target {
   class=FortAthenaBTService_AIEvaluator
   properties {interval=0.5, nodeName=Focus on target, aIEvaluatorClass {defaultValue=&FortAthenaAIBotEvaluator_HandleFocusing}}
  }]
}
